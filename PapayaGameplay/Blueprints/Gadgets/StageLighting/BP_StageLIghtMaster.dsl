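; Stage lighting gadget master that drives three Niagara light cones, paired with either spotlights or decals.
; Setup runs at construction and BeginPlay; outside debug and mobile, it keeps decals for low shading quality or spotlights otherwise, and applies cone dimensions.
; Tick samples a color curve using a shared material parameter collection value plus a per-instance offset, recoloring the active lights or decal materials.
; Custom events set that offset and fade cone alpha with light intensity or decal opacity.
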(event EventBeginPlay
  (CallFunction|CompleteSystemSetup))

(event Custom|Set%20Color%20Offset
  (params (ColorOffset :type Float%20%28double-precision%29))
  (bind _returnvalue (Math|Float|Fraction ColorOffset))
  (Variables|Default|SetColorOffset _returnvalue)
  (for _array_element (Variables|Default|GetLight%20Cones)
    (Niagara|SetNiagaraVariable(Float) _array_element "User.ColorOffset" _returnvalue)))

(event EventTick
  (params (DeltaSeconds :type Float%20%28single-precision%29))
  (bind _coloroffset (Variables|Default|GetColorOffset))
  (bind _basecolorinfo (Variables|Default|GetBaseColorInfo))
  (bind _returnvalue (+ _basecolorinfo _coloroffset))
  (bind _returnvalue_2 (Math|Float|Fraction _returnvalue))
  (bind _basecolorcurve (Variables|Default|GetBaseColorCurve))
  (bind _returnvalue_3 (Math|Curves|GetLinearColorValue _basecolorcurve _returnvalue_2))
  (bind _returnvalue_4 (Rendering|Material|GetScalarParameterValue "/Game/Environments/Papaya/Materials/MPC_PapayaGadgetControls.MPC_PapayaGadgetControls" "Base Color Parameter" :!class /Script/Engine.KismetMaterialLibrary))
  (Variables|Default|SetBaseColorInfo (Math|Float|Fraction _returnvalue_4))
  (if (or (Variables|Default|GetMobileCheck) (Variables|Default|GetQualityCheck))
    (for _array_element (Variables|Default|GetDecalMID)
      (Rendering|Material|SetVectorParameterValue _array_element "LightColor" _returnvalue_3 :!class /Script/Engine.MaterialInstanceDynamic))
    (else
      (for _array_element_2 (Variables|Default|GetSpotLights)
        (Rendering|Components|Light|SetLightColor _array_element_2 _returnvalue_3 true)))))

(event Custom|Fade
  (params (Fades :type Float%20%28double-precision%29))
  (bind _output_get (Variables|Default|SetFadeAlpha Fades))
  (Development|PrintString (Utilities|String|ToString(Float) _output_get))
  (Utilities|FlowControl|Sequence
    (:then_0
      (for _array_element (Variables|Default|GetLight%20Cones)
        (Niagara|SetNiagaraVariable(Float) _array_element "User.Alpha" (Variables|Default|GetFadeAlpha))))
    (:then_1
      (if (or (Variables|Default|GetMobileCheck) (Variables|Default|GetQualityCheck))
        (for _array_element_2 (Variables|Default|GetDecalMID)
          (Rendering|Material|SetScalarParameterValue _array_element_2 "Alpha" (Math|Float|Lerp 0.0 0.25 (Variables|Default|GetFadeAlpha)) :!class /Script/Engine.MaterialInstanceDynamic))
        (else
          (for _array_element_3 (Variables|Default|GetSpotLights)
            (Rendering|Components|Light|SetIntensity _array_element_3 (Math|Float|Lerp 0.0 1.4e+03 (Variables|Default|GetFadeAlpha)))))))))

(fn CompleteSystemSetup
  (Variables|Default|SetLight%20Cones (Utilities|Array|MakeArray (Variables|Default|GetNiagara) (Variables|Default|GetNiagara1) (Variables|Default|GetNiagara2)))
  (if (Variables|Default|GetDEBUG)
    (Variables|Default|SetSpotLights (Utilities|Array|MakeArray (Variables|Default|GetSpotLight) (Variables|Default|GetSpotLight1) (Variables|Default|GetSpotLight2)))
    (Variables|Default|SetDecals (Utilities|Array|MakeArray (Variables|Default|GetDecal) (Variables|Default|GetDecal1) (Variables|Default|GetDecal2)))
    (else
      (bind _output_get (Variables|Default|SetMobileCheck (Fortnite|IsMobilePlatform)))
      (if (not _output_get)
        (CallFunction|GameQualitySettingCheck)
        (if (Variables|Default|GetQualityCheck)
          (for _array_element (Variables|Default|GetSpotLights)
            (Components|DestroyComponent _array_element))
          (Variables|Default|SetDecals (Utilities|Array|MakeArray (Variables|Default|GetDecal) (Variables|Default|GetDecal1) (Variables|Default|GetDecal2)))
          (else
            (for _array_element_2 (Variables|Default|GetDecals)
              (Components|DestroyComponent _array_element_2))
            (Variables|Default|SetSpotLights (Utilities|Array|MakeArray (Variables|Default|GetSpotLight) (Variables|Default|GetSpotLight1) (Variables|Default|GetSpotLight2)))))
        (Utilities|FlowControl|Sequence
          (:then_0
            (bind _returnvalue (Rendering|Material|GetScalarParameterValue "/Game/Environments/Papaya/Materials/MPC_PapayaGadgetControls.MPC_PapayaGadgetControls" "Base Color Parameter" :!class /Script/Engine.KismetMaterialLibrary))
            (Variables|Default|SetBaseColorInfo _returnvalue))
          (:then_1
            (if (or (Variables|Default|GetMobileCheck) (Variables|Default|GetQualityCheck))
              (for (_array_element_3 _array_index) (Variables|Default|GetDecals)
                (Utilities|IsValid _array_element_3
                  (:Is%20Valid
                    (bind _returnvalue_2 (Rendering|Material|CreateDynamicMaterialInstance "/PapayaGameplay/Blueprints/Gadgets/StageLighting/M_LightConeDecal.M_LightConeDecal" (Utilities|String|StringToName (Utilities|String|Append "Decal_MID_" (Utilities|String|ToString(Integer) _array_index))) "None" :!class /Script/Engine.KismetMaterialLibrary))
                    (Rendering|Components|Decal|SetDecalMaterial _array_element_3 _returnvalue_2)
                    (Utilities|Array|Add (Variables|Default|GetDecalMID) _returnvalue_2)
                    (Transformation|SetRelativeLocation _array_element_3 (Math|Vector|MakeVector (* (Variables|Default|GetCone%20Length) 0.5)) false false))))
              (for _array_element_4 (Variables|Default|GetDecalMID)
                (Rendering|Material|SetVectorParameterValue _array_element_4 "LightColor" (Math|Curves|GetLinearColorValue (Variables|Default|GetBaseColorCurve) (Variables|Default|GetBaseColorInfo)) :!class /Script/Engine.MaterialInstanceDynamic))
              (else
                (for _array_element_5 (Variables|Default|GetSpotLights)
                  (Utilities|IsValid _array_element_5
                    (:Is%20Valid
                      (Rendering|Lighting|SetAttenuationRadius _array_element_5 (+ (Variables|Default|GetCone%20Length) 5e+02))
                      (Rendering|Components|Light|SetLightColor _array_element_5 (Math|Curves|GetLinearColorValue (Variables|Default|GetBaseColorCurve) (Variables|Default|GetBaseColorInfo)) true)))))))
          (:then_3
            (for _array_element_6 (Variables|Default|GetLight%20Cones)
              (Niagara|SetNiagaraVariable(Float) _array_element_6 "User.ConeLength" (Variables|Default|GetCone%20Length))
              (Niagara|SetNiagaraVariable(Float) _array_element_6 "User.ConeAngle" (Variables|Default|GetCone%20Angle))
              (Niagara|SetNiagaraVariable(Float) _array_element_6 "User.ConeStartRadius" (Variables|Default|GetCone%20Start%20Radius)))))))))

(fn GameQualitySettingCheck
  (bind _returnvalue (Settings|GetGameUserSettings))
  (Variables|Default|SetQualityCheck (<= (Settings|GetShadingQuality _returnvalue) 1))
  (return))

(fn ConstructionScript
  (CallFunction|CompleteSystemSetup))
