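# M_LightConeDecal is a translucent, unlit deferred decal material for the stage lighting gadget, apparently the soft round pool of a light cone.
# A sphere mask centred in UV space, with adjustable radius and hardness, gives a procedural circle without textures.
# A contrast power shapes the falloff and an Alpha parameter scales it.
# A separate LightColor parameter supplies the tint, white by default.
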Material /PapayaGameplay/Blueprints/Gadgets/StageLighting/M_LightConeDecal
graph {
 properties {
  blendMode=BLEND_Translucent
  materialDomain=MD_DeferredDecal
  pixelDepthOffsetMode=PDOM_Legacy
  shadingModel=MSM_Unlit
  usedShadingModels=Unlit
 }
 nodes [
  n1 = TextureCoordinate()
  n2 = Constant() {r=0.5}
  n3 = ScalarParameter() {defaultValue=0.5, parameterName=Radius}
  n4 = ScalarParameter() {parameterName=Hardness}
  Multiply(
   A=Power(Base=SphereMask(A=@n1, B=@n2, Radius=@n3, Hardness=@n4), Exp=ScalarParameter() {defaultValue=2, parameterName=Contrast})
   B=ScalarParameter() {parameterName=Alpha}
  )
  VectorParameter() {parameterName=LightColor, defaultValue=<rgba 1 1 1 0>}
 ]
}
