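# Additive, two-sided unlit material that fakes a volumetric stage-light cone on Niagara mesh particles, shaped by dynamic parameters.
# A ray-cone intersection function feeds an analytic inverse-square scattering integral in custom HLSL, multiplied by a NaN-safe Schlick phase term for anisotropy.
# An edge fade softens that result before particle color and intensity scale it for emissive, with depth fade on alpha.
# UV1 and vertex color compute radial offsets between two radii.
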Material /PapayaGameplay/Blueprints/Gadgets/StageLighting/M_LightCone_01
graph {
 properties {
  bUseTranslucencyVertexFog=false
  bUsedWithNiagaraMeshParticles=true
  blendMode=BLEND_Additive
  pixelDepthOffsetMode=PDOM_Legacy
  shadingModel=MSM_Unlit
  twoSided=true
  usedShadingModels=Unlit
 }
 nodes [
  n1 = ParticleColor()
  DepthFade(Opacity=@n1.A, FadeDistance=ScalarParameter() {defaultValue=150, parameterName=Depth Fader Distance})
  n2 = DynamicParameter() {parameterIndex=1, defaultValue=<rgba 1 1 1 0.75>, paramNames [R1, Light Shift, EmissiveIntensity, Anistropoy]}
  n3 = Multiply(A=@n1.RGB, B=@n2.EmissiveIntensity)
  MaterialExpressionDotProduct_2 = DotProduct(
   A=Multiply(A=CameraVectorWS()) {constB=-1}
   B=Transform(None=VectorParameter() {parameterName=Light Vector, defaultValue=<rgba 0 0 1 0>}.RGB) {transformSourceType=TRANSFORMSOURCE_Local}
  )
  MaterialExpressionDynamicParameter_5 = DynamicParameter() {parameterIndex=1, defaultValue=<rgba 1 1 1 0.75>, paramNames [R1, Light Shift, EmissiveIntensity, Anistropoy]}
  n4 = Custom!(CosTheta=@MaterialExpressionDotProduct_2, k=@MaterialExpressionDynamicParameter_5.Anistropoy) {
   code="float Inner = (1 + k * CosTheta);\r\n\r\nInner = max(Inner, .001);\r\n\r\nreturn (1 - k * k) / (4 * PI * Inner * Inner);\r\n\r\n"
   description="SchlickPhase (NaN safe)"
   outputType=CMOT_Float1
   inputs [
    {inputName=CosTheta, input {expression=&self:MaterialExpressionDotProduct_2}}
    {
     inputName=k
     input {
      expression=&self:MaterialExpressionDynamicParameter_5
      mask=1
      maskA=1
      outputIndex=3
     }
    }
   ]
  }
  n5 = DynamicParameter() {paramNames [ZMin,ZMax,Tan,R2]}
  n6 = Divide(B=Max(A=@n5.Tan) {constB=0.001})
  MaterialExpressionMaterialFunctionCall_1 = MaterialFunctionCall!(
   Ray Origin
   Ray Direction
   Cone Origin=ParticlePositionWS()
   Cone Z Axis=Transform(None=VectorParameter() {parameterName=Light Vector, defaultValue=<rgba 0 0 1 0>}.RGB) {transformSourceType=TRANSFORMSOURCE_Local}
   Z Min-Max=AppendVector(A=@n5.ZMin, B=@n5.ZMin)
   XYZ Scale=MaterialFunctionCall(X=@n6, Y=@n6, Z=Constant() {r=1}) {materialFunction=&MakeFloat3}.Result
   CUV Vector 1
   CUV Vector 2
  ) {materialFunction=/Game/Effects/Fort_Effects/World/Enviroment/Volumetrics/LightBeams/ConeIntersection}
  n7 = Constant()
  n8 = DynamicParameter() {paramNames [ZMin,ZMax,Tan,R2]}
  n9 = DynamicParameter() {parameterIndex=1, defaultValue=<rgba 1 1 1 0.75>, paramNames [R1, Light Shift, EmissiveIntensity, Anistropoy]}
  MaterialExpressionAdd_1 = Add(
   A=@MaterialExpressionMaterialFunctionCall_1.Local Ray Origin
   B=MaterialFunctionCall(X=@n7, Y=@n7, Z=Multiply(A=@n8.ZMin, B=@n9.Light Shift)) {materialFunction=&MakeFloat3}.Result
  )
  MaterialExpressionScalarParameter_0 = ScalarParameter() {defaultValue=0.0001, desc=Tiny offset to prevent huge brightness values, parameterName=f}
  n10 = Custom!(
   t0=@MaterialExpressionMaterialFunctionCall_1.T0
   t1=@MaterialExpressionMaterialFunctionCall_1.T0
   RayOrigin=@MaterialExpressionAdd_1
   RayDirection=@MaterialExpressionMaterialFunctionCall_1.T0
   f=@MaterialExpressionScalarParameter_0
  ) {
   code="float3 ro = RayOrigin;\r\nfloat3 rd = RayDirection;\r\n\r\nfloat a = dot(rd,rd);\r\nfloat b = dot(ro, rd);\r\nfloat c = dot(ro,ro);\r\nfloat d =  c-b*b;\r\n//Do abs here to make sure d is never -1 for sqrt coming up\r\nd = abs(d);\r\n\r\nt0 = max( t0, 0.0f );\r\n\r\n//tiny offset param to avoid infinite brightness\r\nc+= f;\r\nfloat h2 = sqrt( d );\r\n//Make sure h2 is never exactly 0 to avoid NaN\r\nh2 = max( h2, .001);\r\n\r\nfloat integral_t0 = ( atanFast( ( t0 + b ) / h2 ) ) / h2;\r\nfloat integral_t1 = ( atanFast( ( t1 + b ) / h2 ) ) / h2;\r\n\r\nreturn (integral_t1 - integral_t0);\r\n\r\n\r\n\r\n"
   description="Inverse Square Falloff (With NaN Protection)"
   outputType=CMOT_Float2
   inputs [
    {
     inputName=t0
     input {expression=&self:MaterialExpressionMaterialFunctionCall_1, outputIndex=3}
    }
    {
     inputName=t1
     input {expression=&self:MaterialExpressionMaterialFunctionCall_1, outputIndex=4}
    }
    {inputName=RayOrigin, input {expression=&self:MaterialExpressionAdd_1}}
    {
     inputName=RayDirection
     input {expression=&self:MaterialExpressionMaterialFunctionCall_1, outputIndex=7}
    }
    {inputName=f, input {expression=&self:MaterialExpressionScalarParameter_0}}
   ]
  }
  n11 = Multiply(A=@n4, B=ComponentMask(None=@n10) {r=true})
  n12 = DotProduct(
   A=Normalize(VectorInput=ComponentMask(None=@MaterialExpressionMaterialFunctionCall_1.Ray Entry Local Position) {g=true, r=true})
   B=Normalize(VectorInput=ComponentMask(None=@MaterialExpressionMaterialFunctionCall_1.Local Ray Origin) {g=true, r=true})
  )
  n13 = DynamicParameter() {paramNames [ZMin,ZMax,Tan,R2]}
  n14 = Add(A=ComponentMask(None=@MaterialExpressionMaterialFunctionCall_1.Local Ray Origin) {b=true}, B=@n13.ZMin)
  n15 = Max(A=@n13.Tan) {constB=0.001}
  n16 = Divide(A=ComponentMask(None=@MaterialExpressionMaterialFunctionCall_1.Local Ray Origin) {g=true, r=true}, B=@n15)
  n17 = Divide(A=ScalarParameter() {defaultValue=2, parameterName=Fresnel Fade Distance}, B=@n15)
  n18 = Add(
   A=Subtract(
    A=Max(A=@n14) {constB=0}
    B=MaterialFunctionCall(Vector 2=@n16) {materialFunction=/Engine/Functions/Engine_MaterialFunctions02/Utility/VectorLength}.V2 Length
   )
   B=@n17
  )
  n19 = LinearInterpolate(A=Power(Base=Max(A=@n12) {constB=0.001}), Alpha=Clamp(None=Divide(A=@n18, B=@n17)))
  MaterialFunctionCall(Emissive=Multiply(A=Multiply(A=@n3) {constB=0.5}, B=Multiply(A=@n11, B=@n19))) {materialFunction=&MF_FX_EmissiveAdjustments}
  n20 = DynamicParameter() {parameterIndex=1, defaultValue=<rgba 1 1 1 0.75>, paramNames [R1, Light Shift, EmissiveIntensity, Anistropoy]}
  n21 = TextureCoordinate() {coordinateIndex=1}
  n22 = AppendVector(A=Normalize(VectorInput=Subtract(A=@n21) {constB=0.5}), B=Constant())
  Multiply(
   A=LinearInterpolate(A=@n20.R1, B=DynamicParameter() {paramNames [ZMin,ZMax,Tan,R2]}.R2, Alpha=VertexColor().R)
   B=Normalize(VectorInput=Transform(None=@n22) {transformSourceType=TRANSFORMSOURCE_Local})
  )
 ]
}
