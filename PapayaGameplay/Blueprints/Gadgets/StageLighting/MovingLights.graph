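# Level Sequence that animates a BP_StageLIghtMaster stage lighting rig as a sweeping moving-light show.
# The master actor has a spawn track and holds its position while making one full yaw turn over the keyed range.
# Its three child scene components sit in a triangle, each swinging in roll and yaw through a pattern repeated four times per turn.
# Start and end poses match, so the motion can loop seamlessly.
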LevelSequence /PapayaGameplay/Blueprints/Gadgets/StageLighting/MovingLights
bindings [
 Scene {
  id=#1
  parent_id=#2
  possessed_class=SceneComponent
  tracks [{
    type=MovieScene3DTransformTrack
    sections [{
      type=MovieScene3DTransformSection
      channels [
       Location.X {type=MovieSceneScriptingDoubleChannel, default=0}
       Location.Y {type=MovieSceneScriptingDoubleChannel, default=866}
       Location.Z {type=MovieSceneScriptingDoubleChannel, default=0}
       Rotation.X {
        type=MovieSceneScriptingDoubleChannel
        default=-17.0362
        keys [[0,-15],[30,34.4621],[60,34.6385],[90,-15],[120,34.4621],[150,34.6385],[180,-15],[210,34.4621],[240,34.6385],[270,-15],[300,34.4621],[330,34.6385],[360,-15]]
       }
       Rotation.Y {type=MovieSceneScriptingDoubleChannel, default=1, const [1,13]}
       Rotation.Z {
        type=MovieSceneScriptingDoubleChannel
        default=-2.4874
        keys [[0,0],[30,47.8716],[60,-43.5885],[90,0],[120,47.8716],[150,-43.5885],[180,0],[210,47.8716],[240,-43.5885],[270,0],[300,47.8716],[330,-43.5885],[360,0]]
       }
       Scale.X {type=MovieSceneScriptingDoubleChannel, default=1}
       Scale.Y {type=MovieSceneScriptingDoubleChannel, default=1}
       Scale.Z {type=MovieSceneScriptingDoubleChannel, default=1}
      ]
     }]
   }]
 }
 BP Stage LIght Master {
  id=#2
  possessed_class=BP_StageLIghtMaster_C
  tracks [
   {type=MovieSceneSpawnTrack, sections [{type=MovieSceneSpawnSection, channels [None {type=MovieSceneScriptingBoolChannel, default=true}]}]}
   {
    type=MovieScene3DTransformTrack
    sections [{
      type=MovieScene3DTransformSection
      channels [
       Location.X {type=MovieSceneScriptingDoubleChannel, default=-1023.6748, const [-1023.6748,2]}
       Location.Y {type=MovieSceneScriptingDoubleChannel, default=22690, const [22690,2]}
       Location.Z {type=MovieSceneScriptingDoubleChannel, default=0, const [0,2]}
       Rotation.X {type=MovieSceneScriptingDoubleChannel, default=0, const [0,2]}
       Rotation.Y {type=MovieSceneScriptingDoubleChannel, default=0, const [0,2]}
       Rotation.Z {type=MovieSceneScriptingDoubleChannel, default=0, keys [[0,0],[360,360]]}
       Scale.X {type=MovieSceneScriptingDoubleChannel, default=1}
       Scale.Y {type=MovieSceneScriptingDoubleChannel, default=1}
       Scale.Z {type=MovieSceneScriptingDoubleChannel, default=1}
      ]
     }]
   }
  ]
 }
 Scene2 {
  id=#3
  parent_id=#2
  possessed_class=SceneComponent
  tracks [{
    type=MovieScene3DTransformTrack
    sections [{
      type=MovieScene3DTransformSection
      channels [
       Location.X {type=MovieSceneScriptingDoubleChannel, default=-750}
       Location.Y {type=MovieSceneScriptingDoubleChannel, default=-433}
       Location.Z {type=MovieSceneScriptingDoubleChannel, default=0}
       Rotation.X {
        type=MovieSceneScriptingDoubleChannel
        default=18.1589
        keys [[0,20],[30,-31.8801],[60,-37.6342],[90,20],[120,-31.8801],[150,-37.6342],[180,20],[210,-31.8801],[240,-37.6342],[270,20],[300,-31.8801],[330,-37.6342],[360,20]]
       }
       Rotation.Y {type=MovieSceneScriptingDoubleChannel, default=1, const [1,13]}
       Rotation.Z {
        type=MovieSceneScriptingDoubleChannel
        default=-80.2425
        keys [[0,-75],[30,-16.3927],[60,-95.7374],[90,-75],[120,-16.3927],[150,-95.7374],[180,-75],[210,-16.3927],[240,-95.7374],[270,-75],[300,-16.3927],[330,-95.7374],[360,-75]]
       }
       Scale.X {type=MovieSceneScriptingDoubleChannel, default=1}
       Scale.Y {type=MovieSceneScriptingDoubleChannel, default=1}
       Scale.Z {type=MovieSceneScriptingDoubleChannel, default=1}
      ]
     }]
   }]
 }
 Scene1 {
  id=#4
  parent_id=#2
  possessed_class=SceneComponent
  tracks [{
    type=MovieScene3DTransformTrack
    sections [{
      type=MovieScene3DTransformSection
      channels [
       Location.X {type=MovieSceneScriptingDoubleChannel, default=750}
       Location.Y {type=MovieSceneScriptingDoubleChannel, default=-433}
       Location.Z {type=MovieSceneScriptingDoubleChannel, default=0}
       Rotation.X {
        type=MovieSceneScriptingDoubleChannel
        default=17.0729
        keys [[0,20],[30,-37.2855],[60,-31.4578],[90,20],[120,-37.2855],[150,-31.4578],[180,20],[210,-37.2855],[240,-31.4578],[270,20],[300,-37.2855],[330,-31.4578],[360,20]]
       }
       Rotation.Y {type=MovieSceneScriptingDoubleChannel, default=1, const [1,13]}
       Rotation.Z {
        type=MovieSceneScriptingDoubleChannel
        default=65.4278
        keys [[0,75],[30,94.4429],[60,20.3328],[90,75],[120,94.4429],[150,20.3328],[180,75],[210,94.4429],[240,20.3328],[270,75],[300,94.4429],[330,20.3328],[360,75]]
       }
       Scale.X {type=MovieSceneScriptingDoubleChannel, default=1}
       Scale.Y {type=MovieSceneScriptingDoubleChannel, default=1}
       Scale.Z {type=MovieSceneScriptingDoubleChannel, default=1}
      ]
     }]
   }]
 }
]
display_rate {denominator=1, numerator=30}
tick_resolution {denominator=1, numerator=24000}
tracks []
