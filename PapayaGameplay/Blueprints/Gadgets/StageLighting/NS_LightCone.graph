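# Niagara system for a stage-lighting beam with source flares that react to the jukebox.
# Light_Cone bursts one persistent mesh particle, shaped by user cone parameters, dynamic material values and a local setup module.
# Body_Flare and Cornoa spawn fading camera-offset sprites, tinted by LightColor and sized by the Papaya_Jukebox parameter collection's average amplitude.
# Cone and body flare sample BaseColorCurve at jukebox color plus ColorOffset, while User.Alpha scales all emitters.
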NiagaraSystem /PapayaGameplay/Blueprints/Gadgets/StageLighting/NS_LightCone
dependencies {usedDataInterfaces [&NiagaraDataInterfaceCurve], usedDynamicInputs [], usedModules [&SystemState]}
emitters {
 Body_Flare {
  emitterUpdateScript [
   &EmitterState(Life Cycle Mode=::NewEnumerator1, Loop Duration=1, UseLoopDelay=false)
   &SpawnRate(SpawnRate=2, Use Spawn Probability=false)
  ]
  particleSpawnScript [
   /Niagara/Modules/Spawn/Initialization/InitializeParticle(
    Color=&MakeLinearColorFromVectorAndFloat("Vector (RGB)"=<1 1 1>, "Float (Alpha)"=1)
    Module.Write Color=true
    Module.Write Lifetime=true
    Module.Write Mass=true
    Module.Write Material Random=true
    Module.Write Position=false
    Module.Write Scale=false
    Module.Write SpriteRotation=true
    Module.Write SpriteSize=true
    Position=&SimulationPosition()
    Sprite Rotation=/Niagara/DynamicInputs/Angles/DegreesToNormalizedAngle(AngleInDegrees:NiagaraFloat=&UniformRangedFloat(Minimum=0, Maximum=1))
    Sprite Size=<500 500>
   )
   &CameraOffset(Camera Offset Amount=20)
  ]
  particleUpdateScript [
   &ParticleState()
   &Color(Color=$User.LightColor)
   &ScaleColor(
    Scale Alpha=&FloatFromCurve(
     FloatCurve=DataInterface() {propertyValues {Curve {keys [<at 0 0 RCIM_Cubic>,<at 0.5 1 RCIM_Cubic>,<at 1 0 RCIM_Cubic>]}}}
     Scale Curve=$User.Alpha
    )
    Scale RGB=&MakeVectorFromLinearColorRGB(Color=&ColorFromCurve(
      ColorCurve=$User.BaseColorCurve
      CurveIndex=&Frac_Float(Float=&Add_Float(A=$NPC.Papaya_Jukebox_NPC.Base Color Parameter, B=$User.ColorOffset))
     ))
    ScaleA=true
    ScaleRGB=true
   )
   &SolveForcesAndVelocity(Clamp Velocity=false, Limit Acceleration=false)
   &ScaleSpriteSize(Scale Factor=&Vector2DFromFloat(Value=&Multiply_Float(A=$NPC.Papaya_Jukebox_NPC.AverageAmplitude, B=2)))
  ]
  rendererClasses [&NiagaraSpriteRendererProperties]
  renderers [{rendererClass=&NiagaraSpriteRendererProperties}]
 }
 Cornoa {
  emitterUpdateScript [
   &EmitterState(Life Cycle Mode=::NewEnumerator1, Loop Duration=1, UseLoopDelay=false)
   &SpawnRate(SpawnRate=2, Use Spawn Probability=false)
  ]
  particleSpawnScript [
   &InitializeParticle(
    Color=&MakeLinearColorFromVectorAndFloat("Vector (RGB)"=<1 1 1>, "Float (Alpha)"=0.125)
    Module.Write Color=true
    Module.Write Lifetime=true
    Module.Write Mass=true
    Module.Write Material Random=true
    Module.Write Position=false
    Module.Write Scale=false
    Module.Write SpriteRotation=false
    Module.Write SpriteSize=true
    Position=&SimulationPosition()
    Sprite Size=<400 400>
   )
   &CameraOffset(Camera Offset Amount=20)
   &DynamicMaterialParameters(
    Index 0 Param 1=0.9
    Index 0 Param 2=&UniformRangedFloat(Minimum=-0.5, Maximum=0.5)
    Index 0 Param 3=1
    Module.Param0WriteEnabled=true
    Module.Param1WriteEnabled=true
    Module.Param2WriteEnabled=true
    Module.Param3WriteEnabled=true
   )
  ]
  particleUpdateScript [
   &ParticleState()
   &Color(Color=$User.LightColor)
   &ScaleColor(
    Scale Alpha=&FloatFromCurve(
     FloatCurve=DataInterface() {propertyValues {Curve {keys [<at 0 0 RCIM_Cubic>,<at 0.5 1 RCIM_Cubic>,<at 1 0 RCIM_Cubic>]}}}
     Scale Curve=$User.Alpha
    )
    ScaleA=true
    ScaleRGB=true
   )
   &ScaleSpriteSize(Scale Factor=&Vector2DFromFloat(Value=&Multiply_Float(A=$NPC.Papaya_Jukebox_NPC.AverageAmplitude, B=2)))
   &SolveForcesAndVelocity(Clamp Velocity=false, Limit Acceleration=false)
  ]
  rendererClasses [&NiagaraSpriteRendererProperties]
  renderers [{rendererClass=&NiagaraSpriteRendererProperties}]
 }
 Light_Cone {
  emitterUpdateScript [
   &EmitterState(Life Cycle Mode=::NewEnumerator1, Loop Behavior=::NewEnumerator1, Loop Duration Mode=::NewEnumerator1, UseLoopDelay=false)
   &SpawnBurst_Instantaneous(Spawn Count=1, Spawn Time=0.6, Use Spawn Probability=false)
  ]
  particleSpawnScript [&InitializeParticle(
    Mesh Scale=&MakeVector(X=0.01, Y=0.01, Z=&Divide_Float(A=$User.ConeLength, B=100))
    Module.Write Color=true
    Module.Write Lifetime=true
    Module.Write Mass=true
    Module.Write Material Random=true
    Module.Write Position=false
    Module.Write Scale=true
    Module.Write SpriteRotation=false
    Module.Write SpriteSize=false
    Position=&SimulationPosition()
    Sprite Size=<5 5>
   )]
  particleUpdateScript [
   &ParticleState(Kill Particles When Lifetime Has Elapsed=false, Let Infinitely Lived Particles Die When Emitter Deactivates=true)
   &SolveForcesAndVelocity(Clamp Velocity=false, Limit Acceleration=false)
   &DynamicMaterialParameters(
    Index 1 Param 1=$User.ConeStartRadius
    Index 1 Param 2=0.005
    Index 1 Param 3=5000
    Index 1 Param 4=0.35
    Module.Param0WriteEnabled1=true
    Module.Param1WriteEnabled1=true
    Module.Param2WriteEnabled1=true
    Module.Param3WriteEnabled1=true
    Write Parameter Index 0=false
    Write Parameter Index 1=true
   )
   &self:Light_Cone_Setup() {module=Light_Cone_Setup}
   &ScaleColor(
    Scale Alpha=$User.Alpha
    Scale RGB=&MakeVectorFromLinearColorRGB(Color=&ColorFromCurve(
      ColorCurve=$User.BaseColorCurve
      CurveIndex=&Frac_Float(Float=&Add_Float(A=$NPC.Papaya_Jukebox_NPC.Base Color Parameter, B=$User.ColorOffset))
     ))
    ScaleA=true
    ScaleRGB=true
   )
  ]
  rendererClasses [&NiagaraMeshRendererProperties]
  renderers [{rendererClass=&NiagaraMeshRendererProperties, rendererIndex=0}]
 }
}
summary {
 systemName=NS_LightCone
 emitters [
  {emitterName=Light_Cone, rendererClasses [&NiagaraMeshRendererProperties]}
  {emitterName=Body_Flare, rendererClasses [&NiagaraSpriteRendererProperties]}
  {emitterName=Cornoa, rendererClasses [&NiagaraSpriteRendererProperties]}
 ]
 userVariables [
  User.Alpha {type=NiagaraFloat, defaultValue=0, description=""}
  User.BaseColorCurve {
   type=NiagaraDataInterfaceColorCurve
   description=""
   defaultValue {
    struct=NiagaraExt_VariableValue_DataInterface
    value {
     dataInterface=&self:NiagaraDataInterfaceColorCurve_0
     dataInterfaceClass=&NiagaraDataInterfaceColorCurve
    }
   }
  }
  User.ColorOffset {type=NiagaraFloat, defaultValue=0, description=""}
  User.ConeAngle {type=NiagaraFloat, defaultValue=6.9, description=""}
  User.ConeLength {type=NiagaraFloat, defaultValue=7500, description=""}
  User.ConeStartRadius {type=NiagaraFloat, defaultValue=36, description=""}
  User.LightColor {type=LinearColor, description="", defaultValue {struct=LinearColor, value=<rgba 1 1 1 1>}}
 ]
}
